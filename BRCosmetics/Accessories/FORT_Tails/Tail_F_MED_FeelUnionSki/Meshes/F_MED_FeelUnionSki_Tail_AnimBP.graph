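# Anim Blueprint for the FeelUnionSki tail accessory, posing the tail with chains of static Transform Bone offsets.
# A cape (long backbling) flag blends between the input pose and a tucked StaminaCat sequence, which is reshaped and cached.
# RBAN OFF (DBNO or rigid body disabled) selects between two adjusted versions of that cache.
# On customization load, any cape or backbling sets a material parameter hiding an element.
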AnimBlueprint /BRCosmetics/Accessories/FORT_Tails/Tail_F_MED_FeelUnionSki/Meshes/F_MED_FeelUnionSki_Tail_AnimBP
target_skeleton=/Game/Accessories/FORT_Tails/Common/Fortnite_Base_Tail_Skeleton
graphs [
 AnimGraph {nodes [
   n1 = "Transform (Modify) Bone - Bone: C_Root_Main_Root_Jnt"!(
    ComponentPose=Local To Component(LocalPose=Use cached pose 'iniPoseCache'())
    bAlphaBoolEnabled:bool=True
    Translation="0,5,0"
    Rotation="180,180,180"
   ) -> (Pose) {
    alphaInputType=Bool
    rotationMode=BMM_Replace
    rotationSpace=BCS_ParentBoneSpace
    translationMode=BMM_Replace
    translationSpace=BCS_ParentBoneSpace
    boneToModify {boneName=C_Root_Main_Root_Jnt}
    rotation=<rot 180 180 180>
    translation {y=5}
   }
   n2 = Component To Local(ComponentPose=@n1)
   n3 = "Transform (Modify) Bone - Bone: C_Tail_A_1_Jnt"!(ComponentPose=Local To Component(LocalPose=Use cached pose 'iniPoseCache'()), bAlphaBoolEnabled:bool=True, Rotation="20,0,0") -> (Pose) {
    alphaInputType=Bool
    rotationMode=BMM_Additive
    rotationSpace=BCS_BoneSpace
    alphaBoolBlend {blendOption=Linear}
    boneToModify {boneName=C_Tail_A_1_Jnt}
    rotation {pitch=20}
   }
   n4 = "Transform (Modify) Bone - Bone: C_Tail_A_2_Jnt"!(ComponentPose=@n3, bAlphaBoolEnabled:bool=True, Rotation="30,0,0") -> (Pose) {
    alphaInputType=Bool
    rotationMode=BMM_Additive
    rotationSpace=BCS_BoneSpace
    alphaBoolBlend {blendOption=Linear}
    boneToModify {boneName=C_Tail_A_2_Jnt}
    rotation {pitch=30}
   }
   n5 = "Transform (Modify) Bone - Bone: C_Tail_A_3_Jnt"!(ComponentPose=@n4, bAlphaBoolEnabled:bool=True, Rotation="30,0,0") -> (Pose) {
    alphaInputType=Bool
    rotationMode=BMM_Additive
    rotationSpace=BCS_BoneSpace
    alphaBoolBlend {blendOption=Linear}
    boneToModify {boneName=C_Tail_A_3_Jnt}
    rotation {pitch=30}
   }
   n6 = "Transform (Modify) Bone - Bone: C_Tail_A_5_Jnt"!(ComponentPose=@n5, bAlphaBoolEnabled:bool=True, Rotation="-10,0,0") -> (Pose) {
    alphaInputType=Bool
    rotationMode=BMM_Additive
    rotationSpace=BCS_BoneSpace
    alphaBoolBlend {blendOption=Linear}
    boneToModify {boneName=C_Tail_A_5_Jnt}
    rotation {pitch=-10}
   }
   n7 = "Transform (Modify) Bone - Bone: C_Tail_A_6_Jnt"!(ComponentPose=@n6, bAlphaBoolEnabled:bool=True, Rotation="-15,0,0") -> (Pose) {
    alphaInputType=Bool
    rotationMode=BMM_Additive
    rotationSpace=BCS_BoneSpace
    alphaBoolBlend {blendOption=Linear}
    boneToModify {boneName=C_Tail_A_6_Jnt}
    rotation {pitch=-15}
   }
   n8 = "Transform (Modify) Bone - Bone: C_Tail_A_7_Jnt"!(ComponentPose=@n7, bAlphaBoolEnabled:bool=True, Rotation="-15,0,0") -> (Pose) {
    alphaInputType=Bool
    rotationMode=BMM_Additive
    rotationSpace=BCS_BoneSpace
    alphaBoolBlend {blendOption=Linear}
    boneToModify {boneName=C_Tail_A_7_Jnt}
    rotation {pitch=-15}
   }
   n9 = "Transform (Modify) Bone - Bone: C_Tail_A_8_Jnt"!(ComponentPose=@n8, bAlphaBoolEnabled:bool=True, Rotation="-15,0,0") -> (Pose) {
    alphaInputType=Bool
    rotationMode=BMM_Additive
    rotationSpace=BCS_BoneSpace
    alphaBoolBlend {blendOption=Linear}
    boneToModify {boneName=C_Tail_A_8_Jnt}
    rotation {pitch=-15}
   }
   n10 = "Transform (Modify) Bone - Bone: C_Tail_A_9_Jnt"!(ComponentPose=@n9, bAlphaBoolEnabled:bool=True, Rotation="-15,0,0") -> (Pose) {
    alphaInputType=Bool
    rotationMode=BMM_Additive
    rotationSpace=BCS_BoneSpace
    alphaBoolBlend {blendOption=Linear}
    boneToModify {boneName=C_Tail_A_9_Jnt}
    rotation {pitch=-15}
   }
   n11 = "Transform (Modify) Bone - Bone: C_Tail_A_1_Jnt"(ComponentPose=@n10, bAlphaBoolEnabled=True, Rotation="-2.6,0,0") {
    alphaInputType=Bool
    rotationMode=BMM_Additive
    rotationSpace=BCS_BoneSpace
    translationSpace=BCS_BoneSpace
    alphaBoolBlend {blendInTime=1, blendOption=Linear, blendOutTime=5}
    boneToModify {boneName=C_Tail_A_1_Jnt}
    rotation {pitch=-2.6}
   }
   n12 = "Transform (Modify) Bone - Bone: C_Tail_A_2_Jnt"(ComponentPose=@n11, bAlphaBoolEnabled=True, Rotation="-19.5,0,0") {
    alphaInputType=Bool
    rotationMode=BMM_Additive
    rotationSpace=BCS_BoneSpace
    translationSpace=BCS_BoneSpace
    alphaBoolBlend {blendInTime=1, blendOption=Linear, blendOutTime=5}
    boneToModify {boneName=C_Tail_A_2_Jnt}
    rotation {pitch=-19.5}
   }
   n13 = "Transform (Modify) Bone - Bone: C_Tail_A_6_Jnt"(ComponentPose=@n12, bAlphaBoolEnabled=True, Rotation="-25.8,0,0") {
    alphaInputType=Bool
    rotationMode=BMM_Additive
    rotationSpace=BCS_BoneSpace
    translationSpace=BCS_BoneSpace
    alphaBoolBlend {blendInTime=1, blendOption=Linear, blendOutTime=5}
    boneToModify {boneName=C_Tail_A_6_Jnt}
    rotation {pitch=-25.8}
   }
   n14 = "Transform (Modify) Bone - Bone: C_Tail_A_7_Jnt"(ComponentPose=@n13, bAlphaBoolEnabled=True, Rotation="-13.933,9.178,-1.700") {
    alphaInputType=Bool
    rotationMode=BMM_Additive
    rotationSpace=BCS_BoneSpace
    translationSpace=BCS_BoneSpace
    alphaBoolBlend {blendInTime=1, blendOption=Linear, blendOutTime=5}
    boneToModify {boneName=C_Tail_A_7_Jnt}
    rotation=<rot -13.933 9.178 -1.7>
    translation=<-0.631 3.384 0.129>
   }
   n15 = "Transform (Modify) Bone - Bone: C_Tail_A_8_Jnt"(ComponentPose=@n14, bAlphaBoolEnabled=True, Rotation="-17.647,11.420,-2.244") {
    alphaInputType=Bool
    rotationMode=BMM_Additive
    rotationSpace=BCS_BoneSpace
    translationSpace=BCS_BoneSpace
    alphaBoolBlend {blendInTime=1, blendOption=Linear, blendOutTime=5}
    boneToModify {boneName=C_Tail_A_8_Jnt}
    rotation=<rot -17.647 11.42 -2.244>
   }
   n16 = Component To Local(ComponentPose=@n15)
   n17 = Get RBAN OFF() -> (RBAN OFF:bool)
   n18 = "Transform (Modify) Bone - Bone: C_Tail_A_4_Jnt"!(
    ComponentPose=Local To Component(LocalPose=Blend Poses by bool(BlendPose_0=@n2, BlendPose_1=@n16, bActiveValue=@n17) {blendType=Linear})
    bAlphaBoolEnabled:bool=True
    Rotation="15,0,0"
   ) -> (Pose) {alphaInputType=Bool, rotationMode=BMM_Additive, rotationSpace=BCS_BoneSpace, boneToModify {boneName=C_Tail_A_4_Jnt}, rotation {pitch=15}}
   n19 = "Transform (Modify) Bone - Bone: C_Tail_A_7_Jnt"(ComponentPose=@n18, bAlphaBoolEnabled=True, Rotation="15,0,0") {alphaInputType=Bool, rotationMode=BMM_Additive, rotationSpace=BCS_BoneSpace, boneToModify {boneName=C_Tail_A_7_Jnt}, rotation {pitch=15}}
   Output Pose(Result=Component To Local(ComponentPose=@n19))
   n20 = Input Pose()
   n21 = Random Sequence Player!() -> (Pose:Pose Link Structure) {
    is=title AnimGraphNode_RandomPlayer
    entries [{
      chanceToPlay=1
      maxLoopCount=0
      maxPlayRate=1
      minLoopCount=0
      minPlayRate=1
      sequence=/Game/Accessories/FORT_Tails/Tail_M_MED_StaminaCat/Animations/StaminaCat_Tucked_Tail
      blendIn {blendOption=Linear, blendTime=0.2, customCurve=null}
     }]
   }
   n22 = Get isWearingCape() -> (isWearingCape:bool)
   n23 = "Transform (Modify) Bone - Bone: C_Tail_A_Base_Jnt"!(
    ComponentPose=Local To Component(LocalPose=Blend Poses by bool(BlendPose_0=@n21, BlendPose_1=@n20, bActiveValue=@n22) {blendType=Linear})
    Alpha=1
    Translation="0,0,0"
    Rotation="-30,-90,-90"
   ) -> (Pose) {
    rotationMode=BMM_Replace
    rotationSpace=BCS_ParentBoneSpace
    translationMode=BMM_Replace
    translationSpace=BCS_ParentBoneSpace
    boneToModify {boneName=C_Tail_A_Base_Jnt}
    rotation=<rot -30 -90 -90>
   }
   n24 = "Transform (Modify) Bone - Bone: C_Tail_A_1_Jnt"(ComponentPose=@n23, bAlphaBoolEnabled=True, Rotation="10,0,0") {
    alphaInputType=Bool
    rotationMode=BMM_Additive
    rotationSpace=BCS_BoneSpace
    translationSpace=BCS_BoneSpace
    alphaBoolBlend {blendInTime=0.2, blendOption=Linear, blendOutTime=0.2}
    boneToModify {boneName=C_Tail_A_1_Jnt}
    rotation {pitch=10}
   }
   n25 = "Transform (Modify) Bone - Bone: C_Tail_A_2_Jnt"(ComponentPose=@n24, bAlphaBoolEnabled=True, Rotation="-8,0,0") {
    alphaInputType=Bool
    rotationMode=BMM_Additive
    rotationSpace=BCS_BoneSpace
    translationSpace=BCS_BoneSpace
    alphaBoolBlend {blendInTime=0.2, blendOption=Linear, blendOutTime=0.2}
    boneToModify {boneName=C_Tail_A_2_Jnt}
    rotation {pitch=-8}
   }
   n26 = "Transform (Modify) Bone - Bone: C_Tail_A_7_Jnt"(ComponentPose=@n25, bAlphaBoolEnabled=True, Rotation="20,0,0") {
    alphaInputType=Bool
    rotationMode=BMM_Additive
    rotationSpace=BCS_BoneSpace
    translationSpace=BCS_BoneSpace
    alphaBoolBlend {blendInTime=0.2, blendOption=Linear, blendOutTime=0.2}
    boneToModify {boneName=C_Tail_A_7_Jnt}
    rotation {pitch=20}
    translation=<-0.631 3.384 0.129>
   }
   n27 = "Transform (Modify) Bone - Bone: C_Tail_A_8_Jnt"(ComponentPose=@n26, bAlphaBoolEnabled=True, Rotation="20,0,0") {
    alphaInputType=Bool
    rotationMode=BMM_Additive
    rotationSpace=BCS_BoneSpace
    translationSpace=BCS_BoneSpace
    alphaBoolBlend {blendInTime=0.2, blendOption=Linear, blendOutTime=0.2}
    boneToModify {boneName=C_Tail_A_8_Jnt}
    rotation {pitch=20}
   }
   n28 = "Transform (Modify) Bone - Bone: C_Tail_A_9_Jnt"(ComponentPose=@n27, bAlphaBoolEnabled=True, Rotation="0,10,0") {
    alphaInputType=Bool
    rotationMode=BMM_Additive
    rotationSpace=BCS_BoneSpace
    translationSpace=BCS_BoneSpace
    alphaBoolBlend {blendInTime=0.2, blendOption=Linear, blendOutTime=0.2}
    boneToModify {boneName=C_Tail_A_9_Jnt}
    rotation {yaw=10}
   }
   n29 = "Transform (Modify) Bone - Bone: C_Tail_A_1_Jnt"!(ComponentPose=@n28, Alpha:float=0, Rotation="40,0,0") -> (Pose) {
    rotationMode=BMM_Additive
    rotationSpace=BCS_BoneSpace
    scaleSpace=BCS_BoneSpace
    translationSpace=BCS_BoneSpace
    alphaBoolBlend {blendInTime=0.2, blendOption=Linear, blendOutTime=0.2}
    alphaScaleBiasClamp {bInterpResult=true, interpSpeedDecreasing=5, interpSpeedIncreasing=5}
    boneToModify {boneName=C_Tail_A_1_Jnt}
    rotation {pitch=40}
   }
   n30 = "Transform (Modify) Bone - Bone: C_Tail_A_1_Jnt"!(ComponentPose=@n29, bAlphaBoolEnabled:bool=True, Translation:Vector="0,0,0", Rotation="10,0,0") -> (Pose) {
    alphaInputType=Bool
    rotationMode=BMM_Additive
    rotationSpace=BCS_BoneSpace
    scaleSpace=BCS_BoneSpace
    translationSpace=BCS_BoneSpace
    alphaBoolBlend {blendInTime=0.2, blendOption=Linear, blendOutTime=0.2}
    boneToModify {boneName=C_Tail_A_1_Jnt}
    rotation {pitch=10}
   }
   n31 = "Transform (Modify) Bone - Bone: C_Tail_A_2_Jnt"!(ComponentPose=@n30, bAlphaBoolEnabled=True, Translation:Vector="0,0,0", Rotation="10,0,0") -> (Pose) {
    alphaInputType=Bool
    rotationMode=BMM_Additive
    rotationSpace=BCS_BoneSpace
    scaleSpace=BCS_BoneSpace
    translationSpace=BCS_BoneSpace
    alphaBoolBlend {blendInTime=0.2, blendOption=Linear, blendOutTime=0.2}
    boneToModify {boneName=C_Tail_A_2_Jnt}
    rotation {pitch=10}
   }
   n32 = "Transform (Modify) Bone - Bone: C_Tail_A_3_Jnt"!(ComponentPose=@n31, bAlphaBoolEnabled=True, Translation:Vector="0,0,0", Rotation="5,0,0") -> (Pose) {
    alphaInputType=Bool
    rotationMode=BMM_Additive
    rotationSpace=BCS_BoneSpace
    scaleSpace=BCS_BoneSpace
    translationSpace=BCS_BoneSpace
    alphaBoolBlend {blendInTime=0.2, blendOption=Linear, blendOutTime=0.2}
    boneToModify {boneName=C_Tail_A_3_Jnt}
    rotation {pitch=5}
   }
   n33 = "Transform (Modify) Bone - Bone: C_Tail_A_4_Jnt"(ComponentPose=@n32, bAlphaBoolEnabled=True, Rotation="30,0,0") {
    alphaInputType=Bool
    rotationMode=BMM_Additive
    rotationSpace=BCS_BoneSpace
    scaleMode=BMM_Additive
    scaleSpace=BCS_BoneSpace
    translationSpace=BCS_BoneSpace
    alphaBoolBlend {blendInTime=0.2, blendOption=Linear, blendOutTime=0.2}
    boneToModify {boneName=C_Tail_A_4_Jnt}
    rotation {pitch=30}
   }
   n34 = "Transform (Modify) Bone - Bone: C_Tail_A_5_Jnt"(ComponentPose=@n33, bAlphaBoolEnabled=True, Rotation="30,0,0") {
    alphaInputType=Bool
    rotationMode=BMM_Additive
    rotationSpace=BCS_BoneSpace
    scaleSpace=BCS_BoneSpace
    translationSpace=BCS_BoneSpace
    alphaBoolBlend {blendInTime=0.2, blendOption=Linear, blendOutTime=0.2}
    boneToModify {boneName=C_Tail_A_5_Jnt}
    rotation {pitch=30}
   }
   n35 = "Transform (Modify) Bone - Bone: C_Tail_A_6_Jnt"(ComponentPose=@n34, bAlphaBoolEnabled=True, Rotation="5,0,0") {
    alphaInputType=Bool
    rotationMode=BMM_Additive
    rotationSpace=BCS_BoneSpace
    scaleSpace=BCS_BoneSpace
    translationSpace=BCS_BoneSpace
    alphaBoolBlend {blendInTime=0.2, blendOption=Linear, blendOutTime=0.2}
    boneToModify {boneName=C_Tail_A_6_Jnt}
    rotation {pitch=5}
   }
   n36 = "Transform (Modify) Bone - Bone: C_Tail_A_7_Jnt"(ComponentPose=@n35, bAlphaBoolEnabled=True, Rotation="10,0,0") {
    alphaInputType=Bool
    rotationMode=BMM_Additive
    rotationSpace=BCS_BoneSpace
    scaleSpace=BCS_BoneSpace
    translationSpace=BCS_BoneSpace
    alphaBoolBlend {blendInTime=0.2, blendOption=Linear, blendOutTime=0.2}
    boneToModify {boneName=C_Tail_A_7_Jnt}
    rotation {pitch=10}
   }
   n37 = "Transform (Modify) Bone - Bone: C_Tail_A_8_Jnt"(ComponentPose=@n36, bAlphaBoolEnabled=True, Rotation="0,0,0") {
    alphaInputType=Bool
    rotationMode=BMM_Additive
    rotationSpace=BCS_BoneSpace
    scaleSpace=BCS_BoneSpace
    translationSpace=BCS_BoneSpace
    alphaBoolBlend {blendInTime=0.2, blendOption=Linear, blendOutTime=0.2}
    boneToModify {boneName=C_Tail_A_8_Jnt}
   }
   Save cached pose 'iniPoseCache'(Pose=Component To Local(ComponentPose=@n37)) {cacheName=iniPoseCache}
  ]}
 EventGraph {nodes [
   Event OnCharacterCustomizationLoaded()
   n1 = CheckForBackbling()
   > Set isWearingBackbling(isWearingBackbling=@n1)
   > Set isWearingCape(isWearingCape=@n1.bIsBackblingLong)
   > Branch(Condition=OR Boolean(A=Get isWearingCape(), B=Get isWearingBackbling()))
   > n2 = Get Dynamic Material Instance(SkeletalMeshComponent=GetPartSkeletalMeshComponent())
   > SetScalarParameterValue(self=@n2, ParameterName=Hide Element 1, Value=1)
   Event BlueprintUpdateAnimation()
   > n3 = Sequence()
   n4 = Get bIsMoving2D()
   > Set isOnlyCrouching(isOnlyCrouching:bool=AND Boolean(A=Get bIsCrouching(), B=NOT Boolean(A=@n4))) -> (Output_Get:bool)
   n5 = Get bIsMoving2D()
   n6 = Get isOnlyCrouching() -> (isOnlyCrouching:bool)
   > Set isNotMoving(isNotMoving:bool=AND Boolean(A=NOT Boolean(A=@n5), B=NOT Boolean(A=@n6))) -> (Output_Get:bool)
   n7 = PlayerMovement(execute=@n3.then_1, AnimInstance=Self-Reference(), XDotProductCoefficient=1, YDotProductCoefficient=1, ZDotProductCoefficient=1)
   n8 = BreakVector(InVec=MakeVector(X=@n7, Y=@n7.YMovement, Z=@n7.ZMovement))
   n9 = NOT Boolean(A=Get bIsCrouching())
   n10 = Get bIsOnGround() -> (bIsOnGround:bool)
   > Set y movement(y movement:double=float * float(A=MapRangeClamped(Value=@n8.Y, InRangeA=0, InRangeB=-1), B:bool=AND Boolean(A=@n9, B=@n10)) {is=title K2Node_PromotableOperator}) -> (Output_Get:double)
   n11 = Get bRigidBodyEnabled()
   > Set RBAN OFF(RBAN OFF:bool=OR Boolean(A=Get bIsDBNO(), B=NOT Boolean(A=@n11))) -> (Output_Get:bool)
   TryGetPawnOwner()
  ]}
]
variables [
 isNotMoving {type=boolean}
 isOnlyCrouching {type=boolean}
 isWearingBackbling {type=boolean}
 isWearingCape {type=boolean}
 RBAN OFF {type=boolean}
 y movement {type=number}
]
